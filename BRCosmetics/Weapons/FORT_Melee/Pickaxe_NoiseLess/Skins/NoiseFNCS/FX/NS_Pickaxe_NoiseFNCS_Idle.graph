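# Niagara idle effect for the NoiseFNCS pickaxe skin, playing a sheen and glint on the weapon when not swinging.
# Sharp_Sheen animates a mesh material parameter with shifting colour, while delayed Sharp_Glint pops a scaling sprite at a fixed local offset.
# User.bIsSwinging zeroes burst spawn probability and kills live particles in both emitters.
# Both dim at night through time-of-day scalars; the glint's skeletal mesh location module is disabled.
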NiagaraSystem /BRCosmetics/Weapons/FORT_Melee/Pickaxe_NoiseLess/Skins/NoiseFNCS/FX/NS_Pickaxe_NoiseFNCS_Idle
dependencies {
 usedDataInterfaces [&NiagaraDataInterfaceCurve, &NiagaraDataInterfaceColorCurve]
 usedDynamicInputs [
  &Multiply_Vector2
  &RandomRangeVector2D
  &RandomRangeFloat
 ]
 usedModules [
  &self:SystemSpawnScript.SystemScriptSource.SystemScriptGraph.NiagaraNodeAssignment_0.SetVariables_#1
  &self:SystemSpawnScript.SystemScriptSource.SystemScriptGraph.NiagaraNodeAssignment_2.SetVariables_#2
  &SystemState
 ]
}
emitters {
 Sharp_Glint {
  emitterUpdateScript [
   &EmitterState()
   &SpawnBurst_Instantaneous(
    Spawn Probability=&MakeCustomFloatFromBool(Bool=$User.bIsSwinging, True Float=0, False Float=1)
    Spawn Time=0.35
    Use Spawn Probability=true
    UseLoopCountLimit=false
   )
  ]
  particleSpawnScript [
   &InitializeParticle(
    Color=<rgba 1 0.5791 0.1302 1>
    Lifetime=0.25
    Position Offset=<6.2 0 82.5>
    Position Offset Coordinate Space=::Local
    Sprite Rotation Angle=-30
    Sprite Rotation Mode=::NewEnumerator0
    Sprite Size=<30 50>
    Sprite Size Mode=::NewEnumerator4
    UsePositionOffset=true
    Write Material Random=true
   )
   &SkeletalMeshLocation(
    Apply To Particle Position=false
    Inherit Velocity=false
    MaskSpawnGroup=false
    Mesh Sampling Type=::NewEnumerator5
    Skeletal Mesh=DataInterface() {propertyValues {
      DefaultMesh=null
      ExcludeBoneName=root
      OptionalComponentClass=null
      PreviewMesh=/BRCosmetics/Weapons/FORT_Melee/Pickaxe_Grim_Male/Meshes/Grim_Male_Axe
      SkinningMode=SkinOnTheFly
      SoftSourceActor=null
      SourceMode=Default
      UvSetIndex=0
      WholeMeshLOD=-1
      bExcludeBone=false
      bReadDeformedGeometry=true
      bRequireCurrentFrameData=true
      ComponentTags []
      FilteredBones [b_AR_Root]
      FilteredSockets [AnimTip]
      MeshUserParameter {parameter {name=None, typeDefHandle {registeredTypeIndex=1}, varData []}}
      SamplingRegions []
     }}
    Socket Sampling Mode=::NewEnumerator5
   ) {enabled=false}
   &CameraOffset(Camera Offset Amount=15)
  ]
  particleUpdateScript [
   &ParticleState()
   &SolveForcesAndVelocity(Clamp Velocity=false, Limit Acceleration=false)
   &ScaleSpriteSize(Scale Factor=&Vector2DFromFloat(Value=&FloatFromCurve(FloatCurve=DataInterface() {propertyValues {Curve {keys [
          {arriveTangent=4.6781, interpMode=RCIM_Cubic, leaveTangent=4.6781, tangentMode=RCTM_User, time=0, value=0}
          {interpMode=RCIM_Cubic, tangentMode=RCTM_User, time=0.33, value=1}
          {arriveTangent=-3.0097, interpMode=RCIM_Cubic, leaveTangent=-3.0097, tangentMode=RCTM_User, time=1, value=0}
         ]}}})))
   &ScaleColor(
    Scale Alpha=&FloatFromCurve(FloatCurve=DataInterface() {propertyValues {Curve {keys [
         {interpMode=RCIM_Cubic, tangentMode=RCTM_User, time=0, value=0}
         {interpMode=RCIM_Cubic, tangentMode=RCTM_User, time=0.5, value=1}
         {tangentMode=RCTM_User, time=1, value=0}
        ]}}})
    ScaleA=true
    ScaleRGB=false
   ) {enabled=false}
   &KillParticles(Kill Particles=$User.bIsSwinging)
   &ScaleColor(
    Scale Alpha=&FNTimeOfDayScalar(Night Scale=0.75, Morning Scale=1, Evening Scale=1)
    ScaleA=true
    ScaleRGB=true
   ) {module=ScaleColor001}
  ]
  rendererClasses [&NiagaraSpriteRendererProperties]
  renderers [{rendererClass=&NiagaraSpriteRendererProperties}]
 }
 Sharp_Sheen {
  emitterUpdateScript [
   &EmitterState()
   &SpawnBurst_Instantaneous(
    Spawn Probability=&MakeCustomFloatFromBool(Bool=$User.bIsSwinging, True Float=0, False Float=1)
    Use Spawn Probability=true
    UseLoopCountLimit=false
   )
  ]
  particleSpawnScript [
   &InitializeParticle(Color=<rgba 2 2 2 0.25>, Lifetime=0.4, Mesh Scale Mode=::NewEnumerator3, UsePositionOffset=true, Write Material Random=true)
   &CameraOffset(Camera Offset Amount=0.5)
  ]
  particleUpdateScript [
   &ParticleState()
   &DynamicMaterialParameters(
    Index 0 Param 1=&FloatFromCurve(FloatCurve=DataInterface() {propertyValues {Curve {keys [<at 0 -0.17 RCIM_Cubic>, {arriveTangent=2.3784, tangentMode=RCTM_User, time=1, value=1.72}]}}})
    Index 0 Param 2=-0.5
    Module.Param0WriteEnabled=true
    Module.Param1WriteEnabled=true
    Module.Param2WriteEnabled=false
    Module.Param3WriteEnabled=false
   )
   &ScaleColor(
    Scale Alpha=&FloatFromCurve(
     FloatCurve=DataInterface() {propertyValues {Curve {keys [
         {arriveTangent=3.3224, interpMode=RCIM_Cubic, leaveTangent=3.3224, tangentMode=RCTM_User, time=0, value=0}
         {arriveTangentWeight=0.0534, interpMode=RCIM_Cubic, leaveTangentWeight=0.3735, tangentMode=RCTM_User, time=0.1, value=1}
         {interpMode=RCIM_Cubic, leaveTangent=-2.8748, tangentMode=RCTM_Break, time=0.8, value=1}
         {arriveTangent=-5.9092, interpMode=RCIM_Cubic, leaveTangent=-12.6124, tangentMode=RCTM_Break, time=1, value=0}
        ]}}}
     Scale Curve=&FNTimeOfDayScalar(Night Scale=0.65, Morning Scale=1, Evening Scale=1)
    )
    Scale RGB=&MakeVectorFromLinearColorRGB(Color=&ColorFromCurve(ColorCurve=DataInterface() {propertyValues {
        AlphaCurve {keys [<at 0 1 RCIM_Cubic>]}
        BlueCurve {keys [<at 0.3004 1 RCIM_Cubic>,<at 0.5948 0>,<at 0.748 0.2944 RCIM_Cubic>]}
        GreenCurve {keys [<at 0.3004 0.1828 RCIM_Cubic>,<at 0.5948 0>,<at 0.748 0.6602 RCIM_Cubic>]}
        RedCurve {keys [<at 0.3004 0.3592 RCIM_Cubic>,<at 0.5948 1>,<at 0.748 1 RCIM_Cubic>]}
       }}))
    ScaleA=true
    ScaleRGB=true
   )
   &KillParticles(Kill Particles=$User.bIsSwinging)
  ]
  rendererClasses [&NiagaraMeshRendererProperties]
  renderers [{rendererClass=&NiagaraMeshRendererProperties, rendererIndex=0}]
 }
}
summary {
 systemName=NS_Pickaxe_NoiseFNCS_Idle
 emitters [
  {emitterName=Sharp_Sheen, rendererClasses [&NiagaraMeshRendererProperties]}
  {emitterName=Sharp_Glint, rendererClasses [&NiagaraSpriteRendererProperties]}
 ]
 userVariables [
  User.bIsSwinging {type=NiagaraBool}
  User.Skeletal Mesh {
   type=NiagaraDataInterfaceSkeletalMesh
   defaultValue {
    struct=NiagaraExt_VariableValue_DataInterface
    value {dataInterface=&self:NiagaraDataInterfaceSkeletalMesh_0}
   }
  }
 ]
}
